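# Niagara system for the SolarBeam vehicle drift trail cosmetic, with per-variant painted color palettes.
# Two ribbons trail from the wheels, using age-driven sine waves for wobble and twist plus curves for width and fade.
# Socket-attached glow sprites, curl-noise particles and wheel-wrap meshes complete the effect.
# Vehicle flags feed the system, each emitter masks effects per tire, and the color variant ID adjusts ribbon fade and wrap channel selection.
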NiagaraSystem /VehicleCosmetics/DriftTrail/DriftTrail_SolarBeam/FX/NS_DriftTrail_SolarBeam
dependencies {
 usedDataInterfaces [&NiagaraDataInterfaceCurve, &NiagaraDataInterfaceArrayFloat3, &NiagaraDataInterfaceSkeletalMesh]
 usedDynamicInputs []
 usedModules [
  &self:SystemSpawnScript.SystemScriptSource.SystemScriptGraph.NiagaraNodeAssignment_0.SetVariables_#1
  &FN_UnpackVehicleFlags
  /VehicleCosmetics/FX/Niagara/Modules/VC_SetVehicleSpeedAttribs
  &SystemState
  /Game/Effects/Niagara/Modules/Vehicle/FN_VehicleFlags_Cheatcodes
 ]
}
emitters {
 NE_VC_DriftTrail_Ribbon_Parent {
  emitterUpdateScript [
   &EmitterState()
   &VC_RateQualityHandler(Epic Spawn Rate=64, High Spawn Rate=64, Low Spawn Rate=48, Medium Spawn Rate=48, Use Custom Rate?=true)
   &VC_SpawnRibbon_v1(Number of Ribbons=2, Spawn Time=-1, SpawnRate=$Emitter.Spawn Rate)
  ]
  particleSpawnScript [
   &VC_PaintedVariantsOverride(
    Black=<rgba 0.0171 0.0171 0.018 2>
    BurntSienna=<rgba 0.24 0.0531 0.0074 1>
    Cobalt=<rgba 0.066 0.0838 0.6 1>
    Crimson=<rgba 0.8632 0.0173 0.0201 1>
    DefaultColor=<rgba 1 1 0.34 1>
    ForestGreen=<rgba 0.0533 0.9 0.0469 1>
    Grey=<rgba 0.1 0.1 0.1 0.9>
    Lime=<rgba 0.3024 0.9 0.0254 1>
    Orange=<rgba 0.52 0.1596 0.0052 1.2>
    Pink=<rgba 0.8 0.24 0.8 0.85>
    Purple=<rgba 0.6642 0 0.95 0.4>
    R1 Black:LinearColor=<rgba 0.01 0.01 0.01 2>
    R1 BurntSienna:LinearColor=<rgba 0.15 0.0222 0.0052 1.2>
    R1 Cobalt:LinearColor=<rgba 0.005 0.005 0.5 1.2>
    R1 Crimson:LinearColor=<rgba 0.2 0.0049 0.004 1>
    R1 DefaultColor=<rgba 0.6 0.4777 0.15 1>
    R1 ForestGreen:LinearColor=<rgba 0.0086 0.2 0.0078 1>
    R1 Grey:LinearColor=<rgba 0.0327 0.0327 0.035 1.5>
    R1 Lime:LinearColor=<rgba 0.1163 0.25 0.0069 1.2>
    R1 Orange:LinearColor=<rgba 0.38 0.0606 0.0054 1.2>
    R1 Pink:LinearColor=<rgba 0.9 0.135 0.7129 0.8>
    R1 Purple:LinearColor=<rgba 0.1375 0 0.55 1>
    R1 Saffron:LinearColor=<rgba 0.2 0.2 0.008 1>
    R1 SkyBlue:LinearColor=<rgba 0 0.16 0.9 1>
    R1 TitaniumWhite:LinearColor=<rgba 0.2731 0.2731 0.28 1>
    Saffron=<rgba 0.7 0.7 0.007 1>
    SkyBlue=<rgba 0.01 0.7362 1 1>
    TitaniumWhite=<rgba 0.76 0.76 0.76 1>
    Use R1 Painted Overrides=true
   )
   &VC_InitializeRibbon_v1(
    Color=$Particles.Color
    Custom Offset=true
    Enable Skeletal Position=true
    Lifetime=0.65
    Local Vector Offset=&AddVectorToPosition(Position=<-45 0 -10>, Vector=$User.offset)
    Offset Amount=1
    Ribbon Width=37
   )
   &AddVelocity(
    Velocity=<-1500 0 0>
    Velocity Speed Scale=&MakeCustomFloatFromBool(Bool=$User.bIsFrontendPreview, True Float=1.3, False Float=0.8)
   ) {module=AddVelocity001}
  ]
  particleUpdateScript [
   &FN_DisableTireFX(Mask Index Attribute:NiagaraInt32=$Particles.SpawnGroup, Mask Index FL=2, Mask Index FR=3, Mask Index RL=0, Mask Index RR=1)
   &ParticleState()
   &SolveForcesAndVelocity(Clamp Velocity=false, Limit Acceleration=false)
   &VC_UpdateRibbon_v1(Enable Skeletal Position=true)
   &ScaleColor(
    Scale Alpha=&MakeCustomFloatFromBool(
     Bool=&SetBoolByFloatComparison(A=&MakeFloatFromInt(Int=$User.ColorVariantID), B=2, Comparison Type=::NewEnumerator4)
     True Float=&FloatFromCurve(FloatCurve=DataInterface() {propertyValues {Curve {
         defaultValue=1
         keys [
          {
           arriveTangent=2.1825
           arriveTangentWeight=0.1243
           interpMode=RCIM_Cubic
           leaveTangent=2.1825
           leaveTangentWeight=0.1243
           tangentMode=RCTM_User
           tangentWeightMode=RCTWM_WeightedBoth
           time=0
           value=-0.1
          }
          {arriveTangent=-0.013, interpMode=RCIM_Cubic, leaveTangent=-0.013, tangentMode=RCTM_User, time=0.1617, value=0.9847}
          {
           arriveTangent=0.0238
           arriveTangentWeight=0.3019
           interpMode=RCIM_Cubic
           leaveTangent=0.0238
           leaveTangentWeight=0.3019
           tangentMode=RCTM_User
           tangentWeightMode=RCTWM_WeightedBoth
           time=0.68
           value=0
          }
         ]
        }}})
     False Float=&FloatFromCurve(FloatCurve=DataInterface() {propertyValues {Curve {
         defaultValue=1
         keys [
          {
           arriveTangent=2.1825
           arriveTangentWeight=0.2145
           interpMode=RCIM_Cubic
           leaveTangent=2.1825
           leaveTangentWeight=0.2145
           tangentMode=RCTM_User
           tangentWeightMode=RCTWM_WeightedBoth
           time=0
           value=-0.1
          }
          {arriveTangent=-0.013, interpMode=RCIM_Cubic, leaveTangent=-0.013, tangentMode=RCTM_User, time=0.1617, value=0.9847}
          {
           arriveTangent=0.0238
           arriveTangentWeight=0.3019
           interpMode=RCIM_Cubic
           leaveTangent=0.0238
           leaveTangentWeight=0.3019
           tangentMode=RCTM_User
           tangentWeightMode=RCTWM_WeightedBoth
           time=0.77
           value=0
          }
         ]
        }}})
    )
    ScaleA=true
    ScaleRGB=true
   )
   &ScaleRibbonWidth(Ribbon Width Scale=&FloatFromCurve(FloatCurve=DataInterface() {propertyValues {Curve {
        defaultValue=1
        keys [
         {arriveTangent=1.071, interpMode=RCIM_Cubic, leaveTangent=1.071, tangentMode=RCTM_User, time=0.1, value=0.35}
         {arriveTangent=-0.0504, interpMode=RCIM_Cubic, leaveTangent=-0.0504, tangentMode=RCTM_User, time=0.6, value=0.73}
         {arriveTangent=-1.632, interpMode=RCIM_Cubic, leaveTangent=-1.632, tangentMode=RCTM_User, time=0.8, value=0.5625}
        ]
       }}}))
   /SetVariables(
    Particles.ScaleRibbonWidthOutter:NiagaraFloat=&FloatFromCurve(
     FloatCurve=DataInterface() {propertyValues {Curve {
        defaultValue=1
        keys [
         {arriveTangent=1.071, interpMode=RCIM_Cubic, leaveTangent=1.071, tangentMode=RCTM_User, time=0.1, value=0.35}
         {arriveTangent=-0.0504, interpMode=RCIM_Cubic, leaveTangent=-0.0504, tangentMode=RCTM_User, time=0.6, value=0.62}
         {arriveTangent=-1.632, interpMode=RCIM_Cubic, leaveTangent=-1.632, tangentMode=RCTM_User, time=0.8, value=0.4605}
        ]
       }}}
     Scale Curve=100
    )
    Particles.ScalecolorOutterRibbon:LinearColor=&MakeLinearColorFromVectorAndFloat(
     "Vector (RGB)"=&MakeVectorFromLinearColorRGB(Color=$Particles.R1 Color)
     "Float (Alpha)"=&MakeCustomFloatFromBool(
      Bool=&SetBoolByFloatComparison(A=&MakeFloatFromInt(Int=$System.ColorVariantID), B=2, Comparison Type=::NewEnumerator4)
      True Float=&FloatFromCurve(
       FloatCurve=DataInterface() {propertyValues {Curve {
          defaultValue=1
          keys [
           {
            arriveTangent=5.264
            arriveTangentWeight=0.3718
            interpMode=RCIM_Cubic
            leaveTangent=5.264
            leaveTangentWeight=0.3718
            tangentMode=RCTM_User
            tangentWeightMode=RCTWM_WeightedBoth
            time=0
            value=0
           }
           {
            arriveTangent=-0.0402
            arriveTangentWeight=0.0961
            interpMode=RCIM_Cubic
            leaveTangent=-0.0402
            leaveTangentWeight=0.3915
            tangentMode=RCTM_Break
            tangentWeightMode=RCTWM_WeightedBoth
            time=0.2514
            value=0.8863
           }
           {
            arriveTangent=0.0494
            arriveTangentWeight=0.1466
            interpMode=RCIM_Cubic
            leaveTangent=0.0494
            leaveTangentWeight=0.1466
            tangentMode=RCTM_User
            tangentWeightMode=RCTWM_WeightedBoth
            time=0.72
            value=0
           }
          ]
         }}}
       Scale Curve=&MakeFloatFromLinearColor(LinearColor=$Particles.R1 Color, Channel:ENiagaraFloatFromLinearColorOptions=::NewEnumerator3)
      )
      False Float=&FloatFromCurve(
       FloatCurve=DataInterface() {propertyValues {Curve {
          defaultValue=1
          keys [
           {
            arriveTangent=0.0174
            arriveTangentWeight=0.1004
            interpMode=RCIM_Cubic
            leaveTangent=0.0174
            leaveTangentWeight=0.1004
            tangentMode=RCTM_User
            tangentWeightMode=RCTWM_WeightedBoth
            time=0
            value=0
           }
           {arriveTangent=5.3506, arriveTangentWeight=0.3161, leaveTangent=-0.1339, leaveTangentWeight=0.32, time=0.2514, value=1}
           {arriveTangent=-1.7901, interpMode=RCIM_Cubic, leaveTangent=-1.7901, time=0.45, value=1}
           {
            arriveTangent=-0.2882
            arriveTangentWeight=0.2231
            interpMode=RCIM_Cubic
            leaveTangent=-0.2882
            leaveTangentWeight=0.2231
            tangentMode=RCTM_User
            tangentWeightMode=RCTWM_WeightedBoth
            time=0.81
            value=0
           }
          ]
         }}}
       Scale Curve=&MakeFloatFromLinearColor(LinearColor=$Particles.R1 Color, Channel:ENiagaraFloatFromLinearColorOptions=::NewEnumerator3)
      )
     )
    )
   ) {isSetParameters=true}
   /SetVariables(Particles.Position=&AddVectorToPosition(
     Position=$Particles.Position
     Vector=/VehicleCosmetics/FX/Niagara/Modules/VC_ConvertVectorToWorld(Vector:Vector3f=&MakeVector(
       X=0
       Y=0
       Z=&FloatFromCurve(
        FloatCurve=DataInterface() {propertyValues {Curve {keys [
            <at 0 -1 RCIM_Cubic>
            {arriveTangent=82, interpMode=RCIM_Cubic, leaveTangent=1.8389, tangentMode=RCTM_Break, time=0.2, value=76.0687}
            {arriveTangent=3.1163, interpMode=RCIM_Cubic, leaveTangent=-55.4058, tangentMode=RCTM_Break, time=0.5, value=76.0687}
            {arriveTangent=-14.8689, interpMode=RCIM_Cubic, leaveTangent=-14.8689, tangentMode=RCTM_User, time=0.8, value=0}
           ]}}}
        Scale Curve=0.1
       )
      ))
    )) {isSetParameters=true}
   /SetVariables(Particles.Position=&AddVectorToPosition(
     Position=$Particles.Position
     Vector=&TransformVector(
      Vector=&MakeVector(
       X=0
       Y=&Sine(
        Normalized Angle=&Add_Float(
         A=&Multiply_Float(A=$Particles.NormalizedAge, B=-1.2)
         B=&Multiply_Float(A=$Emitter.LoopedAge, B=2)
        )
        Period=1
        Scale=&FloatFromCurve(
         FloatCurve=DataInterface() {propertyValues {Curve {defaultValue=1, keys [{tangentMode=RCTM_User, time=0, value=0.7}, {tangentMode=RCTM_User, time=0.53, value=1.1}]}}}
         Scale Curve=2
        )
        Bias=0
       )
       Z=&Sine(
        Normalized Angle=&Add_Float(
         A=&Multiply_Float(A=$Particles.NormalizedAge, B=-1.2)
         B=&Multiply_Float(A=$Emitter.LoopedAge, B=2)
        )
        Period=1
        Scale=&FloatFromCurve(
         FloatCurve=DataInterface() {propertyValues {Curve {defaultValue=1, keys [{tangentMode=RCTM_User, time=0, value=0.7}, {tangentMode=RCTM_User, time=0.53, value=1.5}]}}}
         Scale Curve=2
        )
        Bias=0
       )
      )
      Destination Space=::Simulation
     )
    )) {isSetParameters=true}
   /SetVariables(Particles.RibbonFacing=&MaskVectorBySpawnGroup(
     Masked Vector=&MakeVector(
      X=1
      Y=0
      Z=&Sine(
       Normalized Angle=&Add_Float(
        A=&Multiply_Float(A=$Particles.NormalizedAge, B=1)
        B=&Multiply_Float(A=$Emitter.LoopedAge, B=1)
       )
       Period=1
       Scale=&FloatFromCurve(FloatCurve=DataInterface() {propertyValues {Curve {
           defaultValue=1
           keys [{arriveTangent=4.6832, interpMode=RCIM_Cubic, leaveTangent=4.6832, tangentMode=RCTM_User, time=0, value=0}, <at 0.5 1 RCIM_Cubic>]
          }}})
       Bias=0
      )
     )
     Passthrough Vector=&MakeVector(
      X=-1
      Y=0
      Z=&Sine(
       Normalized Angle=&Add_Float(
        A=&Multiply_Float(A=$Particles.NormalizedAge, B=1)
        B=&Multiply_Float(A=$Emitter.LoopedAge, B=1)
       )
       Period=1
       Scale=&FloatFromCurve(FloatCurve=DataInterface() {propertyValues {Curve {
           defaultValue=1
           keys [{arriveTangent=4.6832, interpMode=RCIM_Cubic, leaveTangent=4.6832, tangentMode=RCTM_User, time=0, value=0}, <at 0.5 1 RCIM_Cubic>]
          }}})
       Bias=0
      )
     )
    )) {isSetParameters=true}
  ]
  rendererClasses [&NiagaraRibbonRendererProperties]
  renderers [
   {rendererClass=&NiagaraRibbonRendererProperties}
   {rendererClass=&NiagaraRibbonRendererProperties, rendererIndex=1}
  ]
 }
 NE_VC_DriftTrail_Sprite_Parent {
  emitterUpdateScript [
   &EmitterState(Inactive Response=::NewEnumerator1, Life Cycle Mode=::NewEnumerator1, UseLoopDelay=false)
   &SpawnBurst_Instantaneous(Spawn Count=4, Use Spawn Probability=false, UseLoopCountLimit=false)
   &SpawnRate(SpawnGroup=1, SpawnRate=6, Use Spawn Probability=false)
  ]
  particleSpawnScript [
   /SetVariables(Particles.VisibilityTag=$Emitter.SpawnGroup) {isSetParameters=true}
   &VC_PaintedVariantsOverride(
    Black=<rgba 0.02 0.02 0.02 4>
    BurntSienna=<rgba 0.25 0.0874 0 3>
    Cobalt=<rgba 0.0821 0.075 0.5 4>
    Crimson=<rgba 1 0.1 0.1 1.5>
    DefaultColor=<rgba 1 0.7926 0.1 1.5>
    ForestGreen=<rgba 0.2667 0.8 0 1.5>
    Grey=<rgba 0.09 0.09 0.09 4>
    Lime=<rgba 0.4667 0.8 0 2>
    Orange=<rgba 0.6 0.2117 0 1.5>
    Pink=<rgba 0.99 0.396 0.8375 1>
    Purple=<rgba 0.4104 0 1 2>
    R1 Black:LinearColor=<rgba 0.45 0.4114 0.38 0.6>
    R1 BurntSienna:LinearColor=<rgba 0.65 0.3142 0 1.5>
    R1 Cobalt:LinearColor=<rgba 0.2133 0.2 1 3>
    R1 Crimson:LinearColor=<rgba 1.5 0.4358 0.3119 1>
    R1 DefaultColor=<rgba 3 2.4331 0.54 1>
    R1 ForestGreen:LinearColor=<rgba 0.25 1 0 3>
    R1 Grey:LinearColor=<rgba 0.5 0.5 0.5 0.9>
    R1 Lime:LinearColor=<rgba 0.3927 0.8 0 4>
    R1 Orange:LinearColor=<rgba 1 0.4463 0.1 1.5>
    R1 Pink:LinearColor=<rgba 0.9 0.36 0.727 3>
    R1 Purple:LinearColor=<rgba 0.4996 0.072 0.8 3>
    R1 Saffron:LinearColor=<rgba 0.578 0.5536 0.0867 4>
    R1 SkyBlue:LinearColor=<rgba 0.36 0.9307 1.2 2>
    R1 TitaniumWhite:LinearColor=<rgba 0.85 0.85 0.85 2>
    R2 Black:LinearColor=<rgba 0.3 0.2743 0.2533 0.2>
    R2 BurntSienna:LinearColor=<rgba 0.65 0.3846 0.195 1>
    R2 Cobalt:LinearColor=<rgba 0.7 0.9712 1 0.5>
    R2 Crimson:LinearColor=<rgba 1 0.5233 0.45 0.5>
    R2 DefaultColor:LinearColor=<rgba 4 3.6313 2.4 1>
    R2 ForestGreen:LinearColor=<rgba 0.3309 1 0.2 1>
    R2 Grey:LinearColor=<rgba 0.5 0.5 0.5 1>
    R2 Lime:LinearColor=<rgba 1 1 0 1>
    R2 Orange:LinearColor=<rgba 1 1 0 1>
    R2 Pink:LinearColor=<rgba 1 0.922 1 1>
    R2 Purple:LinearColor=<rgba 0.9921 0.32 1 1>
    R2 Saffron:LinearColor=<rgba 1 0.9924 0.3 0.5>
    R2 SkyBlue:LinearColor=<rgba 0.7611 1.5 1.4866 1>
    Saffron=<rgba 0.8 0.7467 0 2>
    SkyBlue=<rgba 0 0.3333 1 3>
    TitaniumWhite=<rgba 1.2 1.2 1.2 1>
    Use R1 Painted Overrides=true
    Use R2 Painted Overrides=true
   )
   &InitializeParticle(
    Color=$Particles.Color
    Lifetime=&MergeHelper_FloatSwitch(0=5, 1=0.8, 2=?, 3=?, Index=$Emitter.SpawnGroup, Num=::NewEnumerator0)
    Sprite Rotation Mode=::NewEnumerator2
    Sprite Size Mode=::NewEnumerator3
    Uniform Sprite Size=55
    UsePositionOffset=false
    Write Material Random=true
   )
   &PartitionParticles()
   &SkeletalMeshLocation(
    Bone / Socket Index=&Add_Integer(A=$Particles.PartitionParticles.Partition, B=4)
    Bone / Socket Sampling Mode=::NewEnumerator5
    Interpolate Mesh Sampling=false
    MaskSpawnGroup=false
    Mesh Sampling Type=::NewEnumerator6
    Sampled Position Offset=&AddVectorToPosition(Position=<-70 0 -10>, Vector=$User.offset)
    Skeletal Mesh=$System.SkeletalMesh
   )
   &AddVelocity(Velocity=&VC_Mask_Vectors(
     FrontEnd Value:Vector3f=<-1500 0 0>
     blsFrontEndPreview:NiagaraBool=false
     SpeedMin Value:Vector3f=<-800 0 0>
     SpeedMax Value:Vector3f=<-1 0 0>
     Speed Alpha:NiagaraFloat=$System.Speed_Alpha
    )) {enabled=false}
  ]
  particleUpdateScript [
   &FN_DisableTireFX(
    Mask Index Attribute:NiagaraInt32=$Particles.PartitionParticles.Partition
    Mask Index FL=2
    Mask Index FR=3
    Mask Index RL=0
    Mask Index RR=1
   )
   &ParticleState()
   /SetVariables(Particles.CoreSpriteSize:Vector2f=&Multiply_Vector2D_ByFloat(Float=0.45, Vector2D=$Particles.Initial.SpriteSize)) {isSetParameters=true}
   &ScaleSpriteSize(
    Uniform Curve Scale=1.25
    Uniform Curve Sprite Scale=DataInterface() {propertyValues {Curve {
       defaultValue=0
       keys [
        {arriveTangent=4.1785, interpMode=RCIM_Cubic, leaveTangent=4.1785, tangentMode=RCTM_User, time=0, value=0}
        {arriveTangent=1.1179, interpMode=RCIM_Cubic, leaveTangent=1.1179, tangentMode=RCTM_User, time=0.2042, value=0.6228}
        {arriveTangent=0.0224, interpMode=RCIM_Cubic, leaveTangent=0.0224, tangentMode=RCTM_User, time=1, value=1}
       ]
      }}}
   )
   &ScaleColor(
    Scale Alpha=&FloatFromCurve(FloatCurve=DataInterface() {propertyValues {Curve {
        defaultValue=1
        keys [
         {arriveTangent=9.0725, interpMode=RCIM_Cubic, leaveTangent=9.0725, tangentMode=RCTM_User, time=0, value=0}
         {arriveTangent=-2.5249, interpMode=RCIM_Cubic, leaveTangent=-2.5249, tangentMode=RCTM_User, time=0.083, value=0.9847}
         {arriveTangent=-0.0405, interpMode=RCIM_Cubic, leaveTangent=-0.0405, tangentMode=RCTM_User, time=1, value=0}
        ]
       }}})
    ScaleA=true
    ScaleRGB=true
   )
   &SolveForcesAndVelocity(Clamp Velocity=false, Limit Acceleration=false)
   &CameraOffset(Camera Offset Amount=0.1)
  ]
  rendererClasses [&NiagaraSpriteRendererProperties]
  renderers [
   {rendererClass=&NiagaraSpriteRendererProperties}
   {rendererClass=&NiagaraSpriteRendererProperties, rendererIndex=1}
   {rendererClass=&NiagaraSpriteRendererProperties, rendererIndex=2}
  ]
 }
 NE_VC_DriftTrail_Sprite_Parent001 {
  emitterUpdateScript [
   &EmitterState()
   &SpawnRate(Spawn Probability=0.7, SpawnRate=14, Use Spawn Probability=true)
  ]
  particleSpawnScript [
   &VC_PaintedVariantsOverride(
    Black=<rgba 0.035 0.035 0.035 1>
    BurntSienna=<rgba 0.4 0.305 0.02 1>
    Cobalt=<rgba 0.25 0.5811 1 1>
    Crimson=<rgba 0.45 0.1575 0.1654 1>
    DefaultColor=<rgba 0.9781 1 0.1 1>
    ForestGreen=<rgba 0 0.7 0.0463 1>
    Grey=<rgba 0.4 0.4 0.4 1>
    Lime=<rgba 0.4083 0.7 0 1>
    Orange=<rgba 1 0.8395 0.0369 1>
    Pink=<rgba 1 0.5 0.84 1>
    Purple=<rgba 1 0.09 1 1>
    Saffron=<rgba 1 1 0.08 1>
    SkyBlue=<rgba 0.5 1 1 1>
   )
   &InitializeParticle(
    Color=$Particles.Color
    Sprite Rotation Mode=::NewEnumerator2
    Sprite Size Mode=::NewEnumerator3
    Uniform Sprite Size=13
    UsePositionOffset=false
    Write Material Random=true
   )
   &PartitionParticles()
   &SkeletalMeshLocation(
    Bone / Socket Index=&Add_Integer(A=$Particles.PartitionParticles.Partition, B=4)
    Bone / Socket Sampling Mode=::NewEnumerator5
    Interpolate Mesh Sampling=false
    MaskSpawnGroup=false
    Mesh Sampling Type=::NewEnumerator6
    Sampled Position Offset=&AddVectorToPosition(Position=<-60 0 -10>, Vector=$User.offset)
    Skeletal Mesh=$System.SkeletalMesh
   )
   &ShapeLocation(MaskSpawnGroup=false, Override Seed=false, Sphere Radius=6, UseOwnerScale=true)
   &AddVelocity(
    Cone Angle=13
    Module.UseConeFalloff=false
    Module.UseVelDistribution=false
    Override Seed=false
    Velocity Mode=::NewEnumerator2
    Velocity Speed=-500
   )
  ]
  particleUpdateScript [
   &FN_DisableTireFX(
    Mask Index Attribute:NiagaraInt32=$Particles.PartitionParticles.Partition
    Mask Index FL=2
    Mask Index FR=3
    Mask Index RL=0
    Mask Index RR=1
   )
   &ParticleState()
   &ScaleColor(
    Scale Alpha=&FloatFromCurve(FloatCurve=DataInterface() {propertyValues {Curve {
        defaultValue=1
        keys [
         {arriveTangent=-0.0824, interpMode=RCIM_Cubic, leaveTangent=-0.0824, leaveTangentWeight=0.0836, tangentMode=RCTM_User, time=0.65, value=1}
         {arriveTangent=-9.0725, arriveTangentWeight=0.7606, interpMode=RCIM_Cubic, leaveTangent=-9.0725, tangentMode=RCTM_User, time=0.95, value=0}
        ]
       }}})
    ScaleA=true
    ScaleRGB=true
   )
   &ScaleSpriteSize(
    Uniform Curve Scale=0.75
    Uniform Curve Sprite Scale=DataInterface() {propertyValues {Curve {
       defaultValue=1
       keys [
        <at 0 0 RCIM_Cubic>
        {arriveTangent=1.3333, interpMode=RCIM_Cubic, leaveTangent=1.3333, time=0.12, value=1}
        {arriveTangent=-0.0824, interpMode=RCIM_Cubic, leaveTangent=-0.0824, leaveTangentWeight=0.0836, tangentMode=RCTM_User, time=0.75, value=1}
        {arriveTangent=-9.0725, arriveTangentWeight=0.7606, interpMode=RCIM_Cubic, leaveTangent=-9.0725, tangentMode=RCTM_User, time=0.95, value=0}
       ]
      }}}
   )
   &SpriteRotationRate(Delta Time:NiagaraFloat=$Particles.Lifetime, Rotation Rate=20)
   &CurlNoiseForce(Bias Noise=false, Noise Strength=200, Pan Noise=false)
   &Drag(
    Drag=&FloatFromCurve(FloatCurve=DataInterface() {propertyValues {Curve {
        defaultValue=1
        keys [{interpMode=RCIM_Cubic, tangentMode=RCTM_User, time=0, value=0}, {interpMode=RCIM_Cubic, tangentMode=RCTM_User, time=1, value=1}]
       }}})
    Use Linear Drag=true
    Use Rotational Drag=true
   )
   &SolveForcesAndVelocity(Clamp Velocity=false, Limit Acceleration=false)
  ]
  rendererClasses [&NiagaraSpriteRendererProperties]
  renderers [
   {rendererClass=&NiagaraSpriteRendererProperties}
   {rendererClass=&NiagaraSpriteRendererProperties, rendererIndex=1}
  ]
 }
 NE_VC_DriftTrail_WheelWrap_Parent001 {
  emitterSpawnScript [/SetVariables(Emitter.ChannelSelector:Vector3f=&SelectVectorFromArray(
     Direct Array Index=$System.ColorVariantID
     Vector Selection Array=DataInterface() {propertyValues {FloatData [
        <5 2 0>
        <0 0 0>
        <1 4 0>
        <2.5 1 0>
        <7 1.5 0>
        <4 1 0>
        <5.5 1 0>
        <10 1 0>
        <5 1 0>
        <5 1.5 0>
        <5 1.5 0>
        <5.5 1.5 0>
        <6 1.5 0>
        <4.5 1.7 0>
        <4.5 1.3 0>
       ]}}
    )) {isSetParameters=true}]
  emitterUpdateScript [
   &EmitterState(
    Inactive Response=::NewEnumerator1
    Life Cycle Mode=::NewEnumerator1
    Loop Behavior=::NewEnumerator1
    Loop Duration Mode=::NewEnumerator1
    UseLoopDelay=false
   )
   &SpawnBurst_Instantaneous(Spawn Count=2, Use Spawn Probability=false, UseLoopCountLimit=false)
  ]
  particleSpawnScript [
   /SetVariables(Particles.VisibilityTag=$Emitter.SpawnGroup) {isSetParameters=true}
   &VC_PaintedVariantsOverride(
    Black=<rgba 0.05 0.05 0.05 0.2>
    BurntSienna=<rgba 0.8 0.61 0.04 0.2>
    Cobalt=<rgba 0.45 0.6975 1 0.2>
    Crimson=<rgba 0.9 0.18 0.1995 0.2>
    DefaultColor=<rgba 0.901 1 0.103 0.3>
    ForestGreen=<rgba 0 1 0.0661 0.2>
    Grey=<rgba 0.8 0.8 0.8 0.2>
    Lime=<rgba 0.5833 1 0 0.2>
    Orange=<rgba 1 0.8395 0.0369 0.2>
    Pink=<rgba 1 0.55 0.856 0.2>
    Purple=<rgba 1 0.3 1 0.2>
    R1 Black:LinearColor=<rgba 0.01 0.01 0.01 1.5>
    R1 BurntSienna:LinearColor=<rgba 0.45 0.0843 0.0078 1>
    R1 Cobalt:LinearColor=<rgba 0.081 0.039 1.3 0.68>
    R1 Crimson:LinearColor=<rgba 0.6 0.024 0.024 1.1>
    R1 DefaultColor=<rgba 0.9047 0.7346 0.2787 0.55>
    R1 ForestGreen:LinearColor=<rgba 0.2987 0.8 0.048 0.35>
    R1 Gold:LinearColor=<rgba 0.901 1 0.103 0.2>
    R1 Grey:LinearColor=<rgba 0.1 0.1 0.1 1>
    R1 Lime:LinearColor=<rgba 0.2648 0.5885 0 0.68>
    R1 Orange:LinearColor=<rgba 1 0.1473 0 0.5>
    R1 Pink:LinearColor=<rgba 1 0.23 0.8116 0.68>
    R1 Purple:LinearColor=<rgba 0.4458 0.05 1 0.5>
    R1 Saffron:LinearColor=<rgba 0.3 0.3 0 0.7>
    R1 SkyBlue:LinearColor=<rgba 0.081 0.7635 0.9 0.483>
    R1 TitaniumWhite:LinearColor=<rgba 0.75 0.75 0.75 0.68>
    Saffron=<rgba 1 1 0.35 0.2>
    SkyBlue=<rgba 0.75 1.5 1.5 0.2>
    TitaniumWhite=<rgba 0.75 0.75 0.75 0.2>
    Use R1 Painted Overrides=true
   )
   &InitializeParticle(
    Color Mode=::NewEnumerator0
    Lifetime=$Emitter.CurrentLoopDuration
    Mesh Scale Mode=::NewEnumerator4
    Sprite Rotation Mode=::NewEnumerator2
    UsePositionOffset=false
    Write Material Random=true
   )
   &PartitionParticles()
   &SkeletalMeshLocation(
    Bone / Socket Index=&Add_Integer(A=$Particles.PartitionParticles.Partition, B=4)
    Bone / Socket Sampling Mode=::NewEnumerator5
    Interpolate Mesh Sampling=false
    MaskSpawnGroup=false
    Mesh Sampling Type=::NewEnumerator6
    Skeletal Mesh=$System.SkeletalMesh
   )
  ]
  particleUpdateScript [
   &FN_DisableTireFX(
    Mask Index Attribute:NiagaraInt32=$Particles.PartitionParticles.Partition
    Mask Index FL=2
    Mask Index FR=3
    Mask Index RL=0
    Mask Index RR=1
   )
   &VC_WheelWrapScale(Wheel Index Attribute:NiagaraInt32=&Add_Integer(A=$Particles.PartitionParticles.Partition, B=2))
   &ParticleState()
   &UpdateMeshOrientation(Rotation Coordinate Space=::NewEnumerator2, Rotation Vector=<0 1 0>) {enabled=false}
   &SkeletalMeshLocation(
    Bone / Socket Index=&Add_Integer(A=$Particles.PartitionParticles.Partition, B=4)
    Bone / Socket Sampling Mode=::NewEnumerator5
    MaskSpawnGroup=false
    Mesh Sampling Type=::NewEnumerator6
    Skeletal Mesh=$System.SkeletalMesh
   ) {module=SkeletalMeshLocation001}
   &SolveForcesAndVelocity(Clamp Velocity=false, Limit Acceleration=false)
   &DynamicMaterialParameters(
    Index 1 Param 1=0.45
    Module.Index0Float0Write=true
    Module.Index0Float1Write=true
    Module.Index0Float2Write=true
    Module.Index0Float3Write=true
    Module.Index1Float0Write=true
    Module.Index1Float1Write=true
    Module.Index1Float2Write=true
    Module.Index1Float3Write=true
    Write Parameter Index 1=::NewEnumerator1
   )
  ]
  rendererClasses [&NiagaraMeshRendererProperties]
  renderers [
   {rendererClass=&NiagaraMeshRendererProperties, rendererIndex=0}
   {rendererClass=&NiagaraMeshRendererProperties, rendererIndex=1}
  ]
 }
}
summary {
 systemName=NS_DriftTrail_SolarBeam
 emitters [
  {emitterName=NE_VC_DriftTrail_Sprite_Parent, rendererClasses [&NiagaraSpriteRendererProperties]}
  {emitterName=NE_VC_DriftTrail_Sprite_Parent001, rendererClasses [&NiagaraSpriteRendererProperties]}
  {emitterName=NE_VC_DriftTrail_Ribbon_Parent, rendererClasses [&NiagaraRibbonRendererProperties]}
  {emitterName=NE_VC_DriftTrail_WheelWrap_Parent001, rendererClasses [&NiagaraMeshRendererProperties]}
 ]
 userVariables [
  User.bIsFrontendPreview {type=NiagaraBool}
  User.bIsPainted {type=NiagaraBool}
  User.ColorVariantID {type=NiagaraInt32}
  User.GlassColor {type=LinearColor, description="", defaultValue {struct=LinearColor, value=<rgba 1 1 1 1>}}
  User.GlassColorVar {type=NiagaraFloat, defaultValue=1, description=""}
  User.MaxGlassColor {type=NiagaraFloat, defaultValue=5, description=""}
  User.MinimumGlassColor {type=NiagaraFloat, defaultValue=0.5, description=""}
  User.offset {type=Vector3f, description="offset from wheels(set through cosmetic template)", defaultValue {struct=Vector3f, value=<0 0 0>}}
  User.PaintedColor {type=LinearColor, description="", defaultValue {struct=LinearColor, value=<rgba 1 1 1 1>}}
  User.PaintedScalar {type=NiagaraFloat, defaultValue=1, description=""}
  User.VehicleFlags {type=NiagaraInt32, defaultValue=806289408}
  User.Wheel_Radius_Array {
   type=NiagaraDataInterfaceArrayFloat
   defaultValue {
    struct=NiagaraExt_VariableValue_DataInterface
    value {
     dataInterface=&self:NiagaraDataInterfaceArrayFloat_1
     dataInterfaceClass=&NiagaraDataInterfaceArrayFloat
    }
   }
  }
  User.Wheel_Width_Array {
   type=NiagaraDataInterfaceArrayFloat
   defaultValue {
    struct=NiagaraExt_VariableValue_DataInterface
    value {
     dataInterface=&self:NiagaraDataInterfaceArrayFloat_0
     dataInterfaceClass=&NiagaraDataInterfaceArrayFloat
    }
   }
  }
  User.WheelTouchingArray {
   type=NiagaraDataInterfaceArrayFloat
   defaultValue {
    struct=NiagaraExt_VariableValue_DataInterface
    value {
     dataInterface=&self:NiagaraDataInterfaceArrayFloat_2
     dataInterfaceClass=&NiagaraDataInterfaceArrayFloat
    }
   }
  }
 ]
}
